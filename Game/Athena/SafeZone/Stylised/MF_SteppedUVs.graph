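# Material function for the stylised safe zone that turns input UVs into stepped, horizontally panning rows, with parameters named for clouds.
# Ceil of scaled V gives a step index that drives each row's rounded U tiling, offset and pan speed.
# A Frac/Sign test flips U and pan direction on alternate rows, while Gap Scale rescales V within each row.
# It outputs the UVs, step count, a bottom mask and two top masks.
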MaterialFunction /Game/Athena/SafeZone/Stylised/MF_SteppedUVs
graph {nodes [
  n1 = ScalarParameter() {defaultValue=4.2317, parameterName=Cloud Tiling_B}
  n2 = Subtract(
   A=FunctionInput(Preview=TextureCoordinate()) {bUsePreviewValueAsDefault=true, inputName=UVs, inputType=FunctionInput_Vector2}
   B=Constant2Vector() {g=1}
  )
  n3 = ComponentMask(None=@n2) {g=true}
  n4 = FunctionInput(Preview=Constant()) {bUsePreviewValueAsDefault=true, inputName=UVs_Offset, inputType=FunctionInput_Scalar, sortPriority=10}
  n5 = Ceil(None=Add(A=Multiply(A=@n1, B=@n3), B=@n4))
  n6 = AppendVector(A=Multiply(A=ScalarParameter() {defaultValue=36.8073, parameterName=Cloud UV_Offset}, B=@n5), B=Constant())
  n7 = ScalarParameter() {defaultValue=1.5402, parameterName=Cloud Tiling Add}
  n8 = Round(None=Add(A=Multiply(A=@n5, B=@n7), B=ScalarParameter() {defaultValue=9.0771, parameterName=Cloud Tiling}))
  n9 = Add(A=Multiply(A=AppendVector(A=@n8, B=@n1), B=@n2), B=@n4)
  n10 = Divide(A=@n5)
  n11 = Sign(None=Subtract(A=Frac(None=@n10)) {constB=0.25})
  n12 = Subtract(A=Frac(None=ComponentMask(None=@n9) {g=true})) {constB=0.5}
  n13 = ScalarParameter() {defaultValue=1.4306, parameterName=Cloud Gap Scale}
  n14 = AppendVector(A=Multiply(A=ComponentMask(None=@n9) {r=true}, B=@n11), B=Add(A=Divide(A=@n12, B=@n13)) {constB=0.5})
  n15 = ScalarParameter() {defaultValue=0.02, parameterName=Cloud Pan Speed A}
  n16 = ScalarParameter() {defaultValue=0.003, parameterName=Cloud Pan Speed B}
  n17 = ScalarParameter() {defaultValue=0.05, parameterName=Cloud Pan Speed Offset}
  n18 = Multiply(A=Add(A=LinearInterpolate(A=@n15, B=@n16, Alpha=@n5), B=Multiply(A=@n17, B=@n4)), B=@n11)
  n19 = Constant()
  FunctionOutput(None=Panner(Coordinate=Add(A=@n6, B=@n14), Speed=AppendVector(A=@n18, B=@n19))) {outputName=UVs}
  n20 = ComponentMask(None=@n9) {g=true}
  n21 = Ceil(None=@n20)
  FunctionOutput(None=OneMinus(None=Saturate(None=@n21))) {outputName=Bottom_Mask, sortPriority=2}
  FunctionOutput(None=@n21) {outputName=StepCount, sortPriority=3}
  n22 = ScalarParameter() {defaultValue=-3.8675, parameterName=TopMask}
  n23 = Multiply(A=@n4) {constB=2}
  n24 = Add(A=@n22, B=@n23)
  n25 = Subtract(A=@n21, B=@n24)
  FunctionOutput(None=Saturate(None=Ceil(None=@n25))) {outputName=TopMask, sortPriority=2}
  n26 = Ceil(None=Add(A=@n20, B=ScalarParameter() {defaultValue=-0.5, parameterName=teste}))
  FunctionOutput(None=Saturate(None=Subtract(A=@n26, B=@n24))) {outputName=TopMask_2, sortPriority=2}
  Subtract(A=@n22, B=@n23)
 ]}
