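# Material function in the Season 11 safe zone folder that blends four RGBA inputs as vertical bands along a mesh's height.
# Height defaults to bounds-normalized local Z, computed per vertex and interpolated.
# Each band fades linearly from its offset over its scale, with band 3 adding a power curve.
# Static switches zero disabled layers; the sum outputs normalized RGB (vector length, not weight) with saturated alpha.
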MaterialFunction /Game/Athena/SafeZone/S11/MF_4VerticalMasks
graph {nodes [
  n1 = FunctionInput() {bUsePreviewValueAsDefault=true, inputName=Mask0_Texture, inputType=FunctionInput_Vector4, previewValue=<1 0 0 1>}
  n2 = FunctionInput() {bUsePreviewValueAsDefault=true, inputName=Mask0_Offset, inputType=FunctionInput_Scalar, sortPriority=1, previewValue=<1 0 0 1>}
  n3 = MaterialFunctionCall() {materialFunction=&ObjectLocalBounds}
  n4 = Divide(
   A=Subtract(
    A=MaterialFunctionCall() {materialFunction=&LocalPosition}."Local Position (Excluding Offsets)"
    B=@n3.Local Bounds Minimum
   )
   B=@n3.Local Bounds Size
  )
  n5 = FunctionInput(Preview=VertexInterpolator(VS=ComponentMask(None=@n4) {b=true}).PS) {bUsePreviewValueAsDefault=true, inputName=InterpolatedVertexVerticalCoordinates, inputType=FunctionInput_Scalar}
  n6 = FunctionInput() {bUsePreviewValueAsDefault=true, inputName=Mask0_Scale, inputType=FunctionInput_Scalar, sortPriority=2, previewValue=<0.2 0 0 1>}
  n7 = Saturate(None=Divide(A=Distance(A=@n2, B=@n5), B=@n6))
  n8 = Constant4Vector()
  n9 = StaticSwitch(
   True=Multiply(A=@n1, B=OneMinus(None=@n7))
   False=@n8.RGBA
   Value=FunctionInput(Preview=StaticBool() {value=true}) {bUsePreviewValueAsDefault=true, inputName=EnableLayer0, inputType=FunctionInput_StaticBool, sortPriority=13}
  ) {defaultValue=true}
  n10 = FunctionInput() {bUsePreviewValueAsDefault=true, inputName=Mask1_Offset, inputType=FunctionInput_Scalar, sortPriority=4, previewValue=<0.66 0 0 1>}
  n11 = FunctionInput() {bUsePreviewValueAsDefault=true, inputName=Mask1_Scale, inputType=FunctionInput_Scalar, sortPriority=5, previewValue=<0.2 0 0 1>}
  n12 = Saturate(None=Divide(A=Distance(A=@n10, B=@n5), B=@n11))
  n13 = FunctionInput() {bUsePreviewValueAsDefault=true, inputName=Mask1_Texture, inputType=FunctionInput_Vector4, sortPriority=3, previewValue=<0 1 0 1>}
  n14 = StaticSwitch(
   True=Multiply(A=OneMinus(None=@n12), B=@n13)
   False=@n8.RGBA
   Value=FunctionInput(Preview=StaticBool() {value=true}) {bUsePreviewValueAsDefault=true, inputName=EnableLayer1, inputType=FunctionInput_StaticBool, sortPriority=14}
  ) {defaultValue=true}
  n15 = FunctionInput() {bUsePreviewValueAsDefault=true, inputName=Mask2_Offset, inputType=FunctionInput_Scalar, sortPriority=7, previewValue=<0.33 0 0 1>}
  n16 = FunctionInput() {bUsePreviewValueAsDefault=true, inputName=Mask2_Scale, inputType=FunctionInput_Scalar, sortPriority=8, previewValue=<0.2 0 0 1>}
  n17 = Divide(A=Distance(A=@n15, B=@n5), B=@n16)
  n18 = FunctionInput() {bUsePreviewValueAsDefault=true, inputName=Mask2_Texture, inputType=FunctionInput_Vector4, sortPriority=6, previewValue=<0 0 1 1>}
  n19 = Multiply(A=OneMinus(None=Saturate(None=@n17)), B=@n18)
  n20 = FunctionInput(Preview=StaticBool() {value=true}) {bUsePreviewValueAsDefault=true, inputName=EnableLayer2, inputType=FunctionInput_StaticBool, sortPriority=15}
  n21 = FunctionInput() {bUsePreviewValueAsDefault=true, inputName=Mask3_Scale, inputType=FunctionInput_Scalar, sortPriority=11, previewValue=<0.2 0 0 1>}
  n22 = Divide(
   A=Distance(A=FunctionInput() {bUsePreviewValueAsDefault=true, inputName=Mask3_Offset, inputType=FunctionInput_Scalar, sortPriority=10}, B=@n5)
   B=@n21
  )
  n23 = FunctionInput() {bUsePreviewValueAsDefault=true, inputName=Mask3_Pow, inputType=FunctionInput_Scalar, sortPriority=12, previewValue=<1 0 0 1>}
  n24 = Power(Base=OneMinus(None=Saturate(None=@n22)), Exp=@n23)
  n25 = FunctionInput() {bUsePreviewValueAsDefault=true, inputName=Mask3_Texture, inputType=FunctionInput_Vector4, sortPriority=9, previewValue=<1 1 0 0>}
  n26 = StaticBool() {value=true}
  n27 = Add(
   A=Add(A=Add(A=@n9, B=@n14), B=StaticSwitch(True=@n19, False=@n8.RGBA, Value=@n20) {defaultValue=true})
   B=StaticSwitch(
    True=Multiply(A=@n24, B=@n25)
    False=@n8.RGBA
    Value=FunctionInput(Preview=@n26) {bUsePreviewValueAsDefault=true, inputName=EnableLayer3, inputType=FunctionInput_StaticBool, sortPriority=16}
   ) {defaultValue=true}
  )
  n28 = ComponentMask(None=@n27) {b=true, g=true, r=true}
  n29 = ComponentMask(None=@n27) {a=true}
  FunctionOutput(None=AppendVector(A=Normalize(VectorInput=@n28), B=Saturate(None=@n29))) {outputName=Interpolated}
 ]}
