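# Unlit, masked, two-sided material for the stylised storm safe-zone circle disc, with world-space radial cloud texture mapping and vertex-colour-driven radius and thickening math.
# The emissive-labelled reroute blends flat tint and shading colours.
# The graph also holds a volumetric-look storm chain: four panning cloud layers, SH sun lighting, scattering, fog and lightning.
# SafeZone parameter collection values drive cloud speed, height offsets and opacity fading.
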Material /Game/Athena/SafeZone/Stylised/M_StormCircleMain_Disc_WE
graph {
 properties {
  bAllowNegativeEmissiveColor=true
  bCastRayTracedShadows=false
  bFullyRough=true
  bUseLightmapDirectionality=false
  bUsedWithNanite=true
  blendMode=BLEND_Masked
  opacityMaskClipValue=0.35
  pixelDepthOffsetMode=PDOM_Legacy
  shadingModel=MSM_Unlit
  translucencyLightingMode=TLM_SurfacePerPixelLighting
  twoSided=true
  usedShadingModels=Unlit
 }
 nodes [
  n1 = VertexColor()
  n2 = ScalarParameter() {defaultValue=10173.5742, parameterName=SplineThickenWidth}
  n3 = Multiply(A=Sign(None=Subtract(A=@n1.G) {constB=0.5}), B=@n2)
  n4 = VertexColor()
  n5 = LinearInterpolate(
   A=ScalarParameter() {defaultValue=31847.3477, parameterName=TempRadius1}
   B=ScalarParameter() {defaultValue=821201.0625, parameterName=TempRadius2}
   Alpha=Divide(A=Add(A=@n4.R))
  )
  n6 = Normalize(VectorInput=AppendVector(A=LocalPosition().XY, B=Constant()))
  n7 = Add(A=WorldPosition() {worldPositionShaderOffset=WPT_ExcludeAllShaderOffsets}.XYZ, B=Multiply(A=@n5, B=@n6))
  n8 = Subtract(A=@n7, B=ObjectPositionWS() {desc=cCenter})
  n9 = Constant3Vector() {constant=<rgba 1 1 0 0>}
  n10 = Subtract(A=@n7, B=CameraPositionWS())
  n11 = DotProduct(A=Normalize(VectorInput=Multiply(A=@n8, B=@n9)), B=Normalize(VectorInput=Multiply(A=@n10, B=@n9)))
  n12 = OneMinus(None=Divide(A=Add(A=@n11)))
  n13 = Multiply(A=Multiply(A=OneMinus(None=@n1.G), B=Multiply(A=@n2, B=@n12))) {constB=1.9}
  n14 = OneMinus(None=@n11)
  n15 = Divide(A=Add(A=@n14))
  n16 = Multiply(
   A=Power(Base=Saturate(None=@n15), Exp=ScalarParameter() {defaultValue=1, parameterName=TiltTestPow})
   B=ScalarParameter() {defaultValue=-8947.8096, parameterName=Tilt}
  )
  n17 = Sign(None=Subtract(A=VertexColor().G) {constB=0.5})
  n18 = Multiply(A=Add(A=@n5, B=Multiply(A=@n16, B=@n17)), B=@n6)
  Add(A=AppendVector(A=Constant2Vector(), B=Add(A=@n3, B=@n13)), B=@n18)
  Constant2Vector()
  Constant3Vector() {constant=<rgba 0 0 1 0>}
  Constant3Vector() {constant=<rgba 1 0 0 0>}
  Constant3Vector()
  Constant() {r=1}
  n19 = WorldPosition()
  Distance(A=@n19.XY, B=ComponentMask(None=VectorParameter() {parameterName=Center, defaultValue=<rgba 0 0 0 1>}.RGB) {g=true, r=true})
  n20 = WorldPosition()
  Distance(A=@n20.XY, B=ComponentMask(None=VectorParameter() {parameterName=Center, defaultValue=<rgba 0 0 0 1>}.RGB) {g=true, r=true})
  LinearInterpolate()
  LinearInterpolate(A=ScalarParameter() {defaultValue=3728.8877, parameterName=SplineThickenWidth_B}, B=@n2, Alpha=Abs(None=@n11))
  n21 = ScalarParameter() {defaultValue=-0.2057, parameterName=SheerAngleContrast_C}
  n22 = ScalarParameter() {defaultValue=0.4105, parameterName=SheerAngleContrast_D}
  n23 = Saturate(None=Multiply(A=@n11) {constB=-1})
  LinearInterpolate(
   A=Multiply(A=LinearInterpolate(A=@n21, B=@n22, Alpha=@n23)) {constB=-1}
   B=LinearInterpolate(
    A=ScalarParameter() {defaultValue=-0.1436, parameterName=SheerAngleContrast_A}
    B=ScalarParameter() {defaultValue=0.9703, parameterName=SheerAngleContrast_B}
    Alpha=Saturate(None=@n11)
   )
   Alpha=Saturate(None=Sign(None=@n11))
  )
  MaterialFunctionCall(WidthBase=@n2, WidthTip=@n2, WorldPosition=Add(A=WorldPosition().XYZ, B=@n18), Expand U or V UV Channel=StaticBool()) {materialFunction=&SplineThicken}
  n24 = VertexColor()
  n25 = MaterialFunctionCall(Float4=AppendVector(A=@n24, B=@n24)) {materialFunction=&BreakOutFloat4Components}
  n26 = Abs(None=DotProduct(A=VertexNormalWS(), B=Constant3Vector() {constant=<rgba 0 0 1 0>}))
  n27 = Power(Base=ComponentMask(None=VertexInterpolator(VS=@n26).PS) {r=true}, Exp=Constant() {r=2})
  n28 = ScalarParameter() {defaultValue=1, parameterName=BoundsScale}
  n29 = MaterialFunctionCall!(BoundsScale=@n28, CylindricalUVs=NamedRerouteUsage()) {materialFunction=/Game/Athena/SafeZone/S11/MF_CylindricalToRadialUVs}
  n30 = Saturate(None=CollectionParameter() {collection=/Game/Athena/SafeZone/MPC_SafeZoneClouds, parameterName=SafeZoneScale})
  n31 = LinearInterpolate(
   A=ScalarParameter() {defaultValue=1.2, group=StormProgression, parameterName=GlobalSpeedEnd}
   B=ScalarParameter() {defaultValue=0.6, group=StormProgression, parameterName=GlobalSpeedStart}
   Alpha=Power(Base=@n30, Exp=Constant() {r=0.33})
  )
  n32 = Multiply(A=Multiply(A=Time(), B=@n31), B=AppendVector(A=Constant() {r=-0.01}, B=Constant() {r=-0.03}))
  n33 = Constant() {r=5}
  n34 = Constant() {r=1.5052}
  n35 = TextureSample(UVs=Multiply(A=Add(A=@n29.Result, B=@n32), B=AppendVector(A=@n33, B=@n34)), Tex=NamedRerouteUsage()) {samplerType=SAMPLERTYPE_LinearColor}
  n36 = Subtract(A=Multiply(A=@n35.RGB) {constB=2})
  n37 = Multiply(A=Power(Base=@n35.A, Exp=Constant() {r=3}), B=Constant() {r=1.4})
  n38 = MaterialFunctionCall() {materialFunction=&LocalPosition}
  n39 = Subtract(A=WorldPosition().XYZ, B=ObjectPositionWS())
  n40 = Max(A=@n30) {constB=0.01}
  n41 = LinearInterpolate(A=Constant() {r=-1.05}, B=Constant(), Alpha=Power(Base=@n40, Exp=Constant() {r=0.5}))
  n42 = LinearInterpolate(
   A=ScalarParameter() {defaultValue=18, group=WPO, parameterName=BottomScaleEnd}
   B=Constant() {r=2}
   Alpha=Power(Base=@n40, Exp=Constant() {r=1})
  )
  n43 = MaterialFunctionCall() {materialFunction=&LocalPosition}
  n44 = MaterialFunctionCall() {materialFunction=&ObjectLocalBounds}
  n45 = Multiply(A=Divide(A=Subtract(A=@n43."Local Position (Excluding Offsets)", B=@n44.Local Bounds Minimum), B=@n44.Local Bounds Size), B=@n28)
  n46 = Add(A=Multiply(A=@n28) {constB=-0.5}) {constB=0.5}
  n47 = Saturate(None=ComponentMask(None=Add(A=@n45, B=@n46)) {b=true})
  n48 = OneMinus(None=@n47)
  n49 = Constant() {r=10}
  n50 = ComponentMask(None=LinearInterpolate(A=Multiply(A=@n39, B=@n41), B=Multiply(A=@n42, B=@n39), Alpha=Power(Base=@n48, Exp=@n49))) {g=true, r=true}
  n51 = Divide(
   A=Add(A=ComponentMask(None=@n38."Local Position (Excluding Offsets)") {g=true, r=true}, B=@n50) {constB=0}
   B=ComponentMask(None=MaterialFunctionCall() {materialFunction=/Engine/Functions/Engine_MaterialFunctions02/WorldPositionOffset/ObjectScale}.Scale XYZ) {g=true, r=true}
  )
  n52 = Multiply(A=@n31, B=Time())
  n53 = Constant() {r=-0.06}
  n54 = Multiply(A=Rotator(Coordinate=@n51, Time=Multiply(A=@n52, B=@n53) {constB=-1}), B=Divide(A=Constant() {r=0.005}, B=@n42))
  n55 = TextureSample(UVs=ComponentMask(None=VertexInterpolator(VS=@n54).PS) {g=true, r=true}, Tex=NamedRerouteUsage()) {samplerType=SAMPLERTYPE_LinearColor}
  n56 = Multiply(A=@n55.RGB) {constB=2}
  n57 = Constant() {r=1.5}
  n58 = Round(None=Constant() {r=5})
  n59 = LinearInterpolate(A=Constant() {r=1}, B=Constant() {r=0.5}, Alpha=@n30)
  n60 = Multiply(A=Constant() {r=1.5}, B=@n59)
  n61 = NamedRerouteUsage()
  n62 = MaterialFunctionCall(Float2=Multiply(A=@n61, B=MaterialFunctionCall(X=@n58, Y=@n60) {materialFunction=&MakeFloat2}.Result)) {materialFunction=&BreakOutFloat2Components}
  n63 = Constant() {r=-8}
  n64 = LinearInterpolate(A=Constant() {r=0.3}, B=Constant() {r=0.2}, Alpha=Power(Base=@n30, Exp=Constant() {r=0.1}))
  n65 = LinearInterpolate(A=Constant() {r=0.9}, B=Constant() {r=0.6}, Alpha=@n30)
  n66 = Power(Base=Saturate(None=Distance(A=@n62.G, B=@n65)), Exp=Constant() {r=0.2})
  n67 = Multiply(A=Multiply(A=@n62.G, B=Multiply(A=@n63, B=@n64)) {constB=0.1}, B=@n66)
  n68 = Multiply(A=Time(), B=@n31)
  n69 = Multiply(A=@n68, B=Constant() {r=-0.02})
  n70 = TextureSample(
   UVs=MaterialFunctionCall(X=Add(A=@n62.R, B=@n67), Y=Add(A=@n62.G, B=@n69)) {materialFunction=&MakeFloat2}.Result
   Tex=NamedRerouteUsage()
  ) {samplerType=SAMPLERTYPE_LinearColor}
  n71 = Subtract(A=Multiply(A=@n70.RGB) {constB=2})
  n72 = Round(None=Constant() {r=12})
  n73 = Multiply(A=Constant() {r=5}, B=@n59)
  n74 = MaterialFunctionCall(Float2=Multiply(A=@n61, B=MaterialFunctionCall(X=@n72, Y=@n73) {materialFunction=&MakeFloat2}.Result)) {materialFunction=&BreakOutFloat2Components}
  n75 = Constant() {r=-5}
  n76 = Multiply(A=Multiply(A=@n74.G, B=Multiply(A=@n75, B=@n64)) {constB=0.1}, B=@n66)
  n77 = Multiply(A=@n68, B=Constant() {r=-0.08})
  n78 = TextureSample(
   UVs=MaterialFunctionCall(X=Add(A=@n74.R, B=@n76), Y=Add(A=@n74.G, B=@n77)) {materialFunction=&MakeFloat2}.Result
   Tex=NamedRerouteUsage()
  ) {samplerType=SAMPLERTYPE_LinearColor}
  n79 = Subtract(A=Multiply(A=@n78.RGB) {constB=2})
  n80 = Power(Base=@n78.A, Exp=Constant() {r=0.28})
  n81 = MaterialFunctionCall(X=Constant() {r=3}, Y=Constant() {r=0.5}) {materialFunction=&MakeFloat2}
  n82 = NamedRerouteUsage()
  n83 = Multiply(A=Time(), B=@n31)
  n84 = AppendVector(A=Constant() {r=-0.02}, B=Constant() {r=0.03})
  n85 = TextureSample(UVs=Add(A=Multiply(A=@n82, B=@n81.Result), B=Multiply(A=@n83, B=@n84)), Tex=NamedRerouteUsage()) {samplerType=SAMPLERTYPE_LinearColor}
  n86 = AppendVector(A=Subtract(A=Multiply(A=@n85.RGB) {constB=2}), B=@n85.A)
  n87 = Constant() {r=8}
  n88 = Constant() {r=1.6575}
  n89 = MaterialFunctionCall(Float2=Multiply(A=@n82, B=MaterialFunctionCall(X=@n87, Y=@n88) {materialFunction=&MakeFloat2}.Result)) {materialFunction=&BreakOutFloat2Components}
  n90 = Multiply(A=Multiply(A=Time(), B=@n31), B=Constant() {r=0.05})
  n91 = TextureSample(UVs=AppendVector(A=@n89.R, B=Add(A=@n89.G, B=@n90)), Tex=NamedRerouteUsage()) {samplerType=SAMPLERTYPE_LinearColor}
  n92 = Constant() {r=1.7}
  n93 = AppendVector(A=Subtract(A=Multiply(A=@n91.RGB) {constB=2}), B=Multiply(A=Power(Base=@n91.A, Exp=@n92), B=Constant() {r=2}))
  n94 = MaterialFunctionCall!(
   InterpolatedVertexVerticalCoordinates=ComponentMask(None=NamedRerouteUsage()) {g=true}
   Mask0_Texture=Multiply(A=@n27, B=AppendVector(A=@n36, B=@n37))
   Mask0_Offset=Constant() {r=0.52}
   Mask0_Scale=Constant() {r=0.2}
   Mask1_Texture=Multiply(A=Multiply(A=@n86, B=@n93), B=OneMinus(None=@n27))
   Mask1_Offset=Constant() {r=0.7552}
   Mask1_Scale=Constant() {r=0.2876}
   Mask2_Texture=Multiply(A=AppendVector(A=@n71, B=@n70.A), B=AppendVector(A=@n79, B=@n80))
   Mask2_Offset=Constant() {r=0.2919}
   Mask2_Scale=Constant() {r=0.22}
   Mask3_Texture=AppendVector(A=Subtract(A=@n56), B=Multiply(A=@n55.A, B=@n57))
   Mask3_Offset=Constant()
   Mask3_Scale=Constant() {desc=0.25, r=0.25}
   Mask3_Pow=Constant() {desc=2.5, r=2.5}
   EnableLayer0=StaticBoolParameter() {defaultValue=true, parameterName=EnableLayer_0}
   EnableLayer1=StaticBoolParameter() {defaultValue=true, parameterName=EnableLayer_1}
   EnableLayer2=StaticBoolParameter() {defaultValue=true, parameterName=EnableLayer_2}
   EnableLayer3=StaticBoolParameter() {defaultValue=true, parameterName=EnableLayer_3}
  ) {materialFunction=/Game/Athena/SafeZone/S11/MF_4VerticalMasks}
  n95 = NamedRerouteDeclaration(None=ComponentMask(None=@n94.Interpolated) {a=true}) {name=StormAlpha, nodeColor=<rgba 1 0 0.4922 1>}
  n96 = Multiply(A=@n25.R, B=OneMinus(None=@n95))
  n97 = Constant4Vector() {constant=<rgba 1.6 0.5 0.5 0.5>}
  n98 = Constant4Vector() {constant=<rgba 1.2 -0.5 -0.5 -0.5>}
  MaterialExpressionAdd_23 = Add(
   A=Multiply(
    A=MaterialFunctionCall(X=@n96, Y=@n25.G, Z=@n25.G, A=@n25.G) {materialFunction=&MakeFloat4}.Result
    B=Subtract(A=@n97.RGBA, B=@n98.RGBA)
   )
   B=@n98.RGBA
  )
  n99 = SkyAtmosphereLightDirection()
  n100 = MaterialFunctionCall(Float3=Multiply(
    A=Transform(None=@n99) {transformSourceType=TRANSFORMSOURCE_World, transformType=TRANSFORM_Local}
    B=Constant3Vector() {constant=<rgba 1 1 -1 0>}
   ) {constB=-1}) {materialFunction=&BreakOutFloat3Components}
  MaterialExpressionVertexInterpolator_3 = VertexInterpolator(VS=MaterialFunctionCall(X=@n100.R, Y=@n100.R, Z=@n100.R) {materialFunction=&MakeFloat3}.Result)
  n101 = Custom!(SH=@MaterialExpressionAdd_23, Direction=@MaterialExpressionVertexInterpolator_3.PS) {
   code="//SH decode,  Direction = LightDir, SH = 2band coefficients. (encoded in Y-up)\r\nDirection = normalize(Direction);\r\nfloat4 shEvalDirection;\r\nshEvalDirection.x = 0.282095;\r\nshEvalDirection.y = -0.488603 * Direction.y;\r\nshEvalDirection.z = 0.488603 * Direction.z;\r\nshEvalDirection.w = -0.488603 * Direction.x;\r\nreturn dot(SH, shEvalDirection);"
   description=SphericalHarmonicsEval
   outputType=CMOT_Float1
   inputs [
    {inputName=SH, input {expression=&self:MaterialExpressionAdd_23}}
    {inputName=Direction, input {expression=&self:MaterialExpressionVertexInterpolator_3}}
   ]
  }
  n102 = Power(Base=OneMinus(None=@n101), Exp=Constant() {r=13})
  n103 = Constant() {r=60}
  n104 = Constant() {r=0.1}
  n105 = NamedRerouteUsage()
  n106 = Saturate(None=Multiply(A=Multiply(A=@n102, B=@n103), B=Power(Base=@n105, Exp=@n104)))
  n107 = Constant() {r=3.8}
  n108 = Saturate(None=Multiply(A=Power(Base=@n101, Exp=@n107) {constExponent=5}, B=Constant() {r=8}) {constB=2})
  n109 = Constant() {r=0.4}
  n110 = VertexNormalWS()
  n111 = MaterialFunctionCall(Float3=NamedRerouteDeclaration(None=ComponentMask(None=@n94.Interpolated) {b=true, g=true, r=true}) {name=StormNormal, nodeColor=<rgba 0 0.9062 1 1>}) {materialFunction=&BreakOutFloat3Components}
  n112 = Multiply(A=@n111.B, B=Constant() {r=0.3})
  n113 = MaterialFunctionCall(
   BaseNormal=Transform(None=@n110) {transformSourceType=TRANSFORMSOURCE_World, transformType=TRANSFORM_Tangent}
   AdditionalNormal=Normalize(VectorInput=MaterialFunctionCall(X=@n111.R, Y=@n111.R, Z=@n112) {materialFunction=&MakeFloat3}.Result)
  ) {materialFunction=&BlendAngleCorrectedNormals}
  n114 = DotProduct(A=QualitySwitch(Default=@n110, Epic=Transform(None=@n113.Result)), B=SkyAtmosphereLightDirection())
  n115 = Add(A=Multiply(A=Saturate(None=@n114), B=@n108), B=OneMinus(None=@n108))
  n116 = Min(
   A=Desaturation(
    None=CollectionParameter() {collection=&FortniteMaterialParameters, parameterName=SunLightColor}
    Fraction=Constant() {r=1}
   )
   B=Constant() {r=2.9}
  )
  n117 = Power(Base=ComponentMask(None=NamedRerouteUsage()) {g=true}, Exp=Constant() {r=1.7})
  n118 = Constant() {r=10}
  n119 = Multiply(A=Multiply(A=Power(Base=@n115), B=@n116), B=Saturate(None=Multiply(A=@n117, B=@n118)))
  n120 = Multiply(A=@n109, B=@n119)
  n121 = VectorParameter() {group=2_BackScattering, parameterName=BackScatteringColor, defaultValue=<rgba 0.1119 0.0433 0.3125 1>}
  n122 = Constant3Vector() {constant=<rgba 0.5417 0.3068 0.2285 1>}
  n123 = CollectionParameter() {collection=&MPC_SafeZoneClouds, parameterName=LocalColorPosition}
  n124 = Power(Base=SphereMask(
    A=ComponentMask(None=@n123) {b=true, g=true, r=true}
    B=WorldPosition().XYZ
    Radius=CollectionParameter() {collection=&MPC_SafeZoneClouds, parameterName=LocalColorRadius}
   ) {attenuationRadius=250000, hardnessPercent=65}) {constExponent=0.7}
  n125 = Multiply(A=Multiply(A=@n106, B=LinearInterpolate(A=@n121.RGB, B=@n122, Alpha=@n124)), B=@n116)
  n126 = Constant3Vector()
  n127 = Multiply(A=ComponentMask(None=SkyAtmosphereLightDirection()) {b=true, g=true, r=true}) {constB=-1}
  n128 = CrossProduct(A=@n127, B=Constant3Vector() {constant=<rgba 0 0 -1 0>})
  n129 = Normalize(VectorInput=CrossProduct(A=@n128, B=@n127))
  n130 = OneMinus(None=@n47)
  n131 = Constant() {r=-0.85}
  n132 = Saturate(None=Multiply(A=Add(A=@n130, B=@n131) {constB=-0.5}, B=Constant() {r=100}))
  n133 = ComponentMask(None=MaterialFunctionCall() {materialFunction=&LocalPosition}.Local Position) {g=true, r=true}
  n134 = AppendVector(A=Constant() {r=30}, B=Constant() {r=40})
  n135 = Time()
  n136 = Constant() {r=0.07}
  n137 = Sine(None=Subtract(A=Divide(A=@n133, B=@n134), B=Multiply(A=@n135, B=@n136)))
  n138 = ComponentMask(None=@n137) {r=true}
  n139 = ComponentMask(None=@n137) {g=true}
  n140 = Constant() {r=3000}
  n141 = Add(A=Multiply(A=OneMinus(None=@n30), B=Constant() {r=0.01}) {constB=4})
  n142 = Multiply(A=Constant3Vector() {constant=<rgba 0.2 0.2 1 1>}, B=Multiply(A=Add(A=@n138, B=@n139), B=Divide(A=@n140, B=@n141)))
  n143 = Constant3Vector() {constant=<rgba 0 0 40000 0>}
  n144 = Power(Base=@n30, Exp=Constant() {r=0.1})
  n145 = Power(Base=@n47, Exp=Add(A=Multiply(A=@n144) {constB=0.8}) {constB=0.2})
  n146 = Power(Base=@n47, Exp=Constant() {r=-3.2656})
  n147 = ScalarParameter() {defaultValue=0.2226, parameterName=TestHeightFalloff3_Bias}
  n148 = Power(Base=Saturate(None=Subtract(A=@n47, B=@n147)), Exp=Constant() {r=0.7795})
  n149 = WorldPosition()
  n150 = Constant2Vector()
  n151 = Multiply(A=Power(Base=@n30, Exp=Constant() {r=0.4})) {constB=2}
  n152 = CollectionParameter() {collection=&MPC_SafeZoneClouds, parameterName=CloudStartingBias}
  n153 = CollectionParameter() {collection=&MPC_SafeZoneClouds, parameterName=PreventStormShrinking}
  n154 = OneMinus(None=Saturate(None=LinearInterpolate(A=@n152, Alpha=@n153) {constA=1, constAlpha=1}))
  n155 = Multiply(A=@n149.XYZ, B=Multiply(A=AppendVector(A=@n150, B=@n151), B=@n154))
  n156 = Add(
   A=Multiply(A=Multiply(A=@n146, B=@n148), B=ScalarParameter() {defaultValue=1607.6896, parameterName=Push Up Dome})
   B=ComponentMask(None=Subtract(A=@n155, B=@n149.XYZ)) {b=true}
  )
  n157 = Add(
   A=Add(A=AppendVector(A=@n50, B=@n156), B=Multiply(A=@n143, B=@n145))
   B=StaticSwitch(True=Multiply(A=@n132, B=@n142), False=Constant(), Value=StaticBoolParameter() {defaultValue=true, parameterName=EnableSineWaveWPO})
  )
  n158 = Add(A=WorldPosition().XYZ, B=@n157) {constB=0}
  n159 = Normalize(VectorInput=@n128)
  n160 = Divide(A=AppendVector(A=DotProduct(A=@n129, B=@n158), B=DotProduct(A=@n159, B=@n158)), B=Constant() {r=700000})
  n161 = Multiply(A=@n160, B=@n160)
  n162 = OneMinus(None=Add(A=ComponentMask(None=@n161) {r=true}, B=ComponentMask(None=@n161) {g=true}))
  n163 = Constant() {r=100}
  n164 = Add(A=Multiply(A=Constant() {r=12}, B=Power(Base=@n162, Exp=@n163)), B=@n162)
  n165 = Multiply(A=@n95, B=@n95)
  n166 = VectorParameter() {parameterName=SunScatteringMorning, defaultValue=<rgba 0.5529 0.0902 0.0706 1>}
  n167 = Constant3Vector() {constant=<rgba 1 0 0.0462 1>}
  n168 = Multiply(
   A=Multiply(A=Max(A=@n164) {constB=0}, B=@n165)
   B=MaterialFunctionCall(
    Color_Morning=LinearInterpolate(A=@n166.RGB, B=@n167, Alpha=@n124)
    Color_Day=VectorParameter() {parameterName=SunScatteringDay, defaultValue=<rgba 0.1502 0.2683 0.6562 1>}.RGB
    Color_Evening=VectorParameter() {parameterName=SunScatteringEvening, defaultValue=<rgba 1 0.1673 0.1291 1>}.RGB
    Color_Night=VectorParameter() {parameterName=SunScatteringNight, defaultValue=<rgba 0.0248 0.0345 0.07 1>}.RGB
   ) {materialFunction=&Color_Time_of_Day}.Result
  )
  n169 = VectorParameter() {group=Fake Layered Fog, parameterName=FogColor, defaultValue=<rgba 0.1453 0.1581 0.3281 0>}
  n170 = Constant3Vector() {constant=<rgba 0.5938 0.3321 0.2598 1>}
  n171 = VertexInterpolator(VS=Power(Base=OneMinus(None=@n47), Exp=Constant() {r=50}))
  n172 = StaticSwitch(
   True=Multiply(A=LinearInterpolate(A=@n169.RGB, B=@n170, Alpha=@n124), B=ComponentMask(None=@n171.PS) {r=true})
   False=Constant()
   Value=StaticBoolParameter() {defaultValue=true, parameterName=EnableBottomFog}
  )
  n173 = LinearInterpolate(
   A=VectorParameter() {group=0_CloudColors, parameterName=CloudColorDark, defaultValue=<rgba 0.0729 0.0424 0.7396 1>}.RGB
   B=Constant3Vector() {constant=<rgba 0.6667 0.3259 0.2222 1>}
   Alpha=@n124
  )
  n174 = LinearInterpolate(
   A=VectorParameter() {group=0_CloudColors, parameterName=CloudColorLight, defaultValue=<rgba 0.6672 0.2702 0.8646 1>}.RGB
   B=Constant3Vector() {constant=<rgba 0.7917 0.3135 0.2062 1>}
   Alpha=@n124
  )
  n175 = Power(Base=@n105, Exp=Constant() {r=2.2})
  n176 = Multiply(A=Multiply(A=LinearInterpolate(A=@n173, B=@n174, Alpha=@n175), B=Multiply(A=@n119, B=@n175)), B=@n109)
  n177 = ScalarParameter() {defaultValue=1.2093, parameterName=SecondaryCloudsMaskBias}
  n178 = ConstantBiasScale(None=Cosine(None=Add(A=@n105, B=@n177)))
  n179 = ScalarParameter() {defaultValue=0.732, parameterName=SecondaryCloudsMultiplier}
  n180 = Multiply(
   A=NamedRerouteUsage()
   B=Power(Base=Multiply(A=@n178, B=@n179) {constB=0.3}, Exp=ScalarParameter() {defaultValue=1.5, parameterName=SecondaryCloudsExponent}) {constExponent=1}
  )
  n181 = VectorParameter() {parameterName=SecondaryClouds}
  n182 = Power(Base=@n105, Exp=Constant() {r=5})
  n183 = Constant() {r=30}
  n184 = Subtract(A=CameraPositionWS(), B=WorldPosition().XYZ)
  n185 = Saturate(None=DotProduct(A=VertexNormalWS(), B=Normalize(VectorInput=@n184)))
  n186 = DotProduct(A=VertexNormalWS(), B=Constant3Vector() {constant=<rgba 0 0 1 0>})
  n187 = VertexInterpolator(VS=Multiply(A=OneMinus(None=@n185), B=Saturate(None=@n186)))
  n188 = Multiply(
   A=Multiply(A=Multiply(A=@n182, B=@n183), B=ComponentMask(None=@n187.PS) {r=true}) {constA=1}
   B=LinearInterpolate(
    A=VectorParameter() {group=Fresnel, parameterName=FresnelColor, defaultValue=<rgba 0.0498 0.0322 0.1094 0>}.RGB
    B=Constant3Vector() {constant=<rgba 0.1875 0.1177 0.0479 1>}
    Alpha=@n124
   )
  )
  n189 = SkyAtmosphereLightDirection()
  n190 = MaterialFunctionCall() {materialFunction=&CameraDirectionVector}
  n191 = OneMinus(None=Saturate(None=DotProduct(A=@n189, B=@n190.Result)))
  n192 = Add(
   A=Add(A=Add(A=@n176, B=@n125) {constB=0}, B=Multiply(A=@n180, B=@n181.RGB)) {constB=0}
   B=QualitySwitch(Default=Constant3Vector(), Epic=Multiply(A=@n188, B=@n191))
  ) {constB=0}
  n193 = Constant3Vector()
  n194 = Multiply(A=Time(), B=Constant() {r=2.8429})
  n195 = Frac(None=@n194)
  n196 = Multiply(A=@n195, B=OneMinus(None=@n195))
  n197 = Constant() {r=1}
  n198 = WorldPosition()
  n199 = Ceil(None=@n194)
  n200 = Constant() {r=0.1062}
  n201 = TextureSampleParameter2D(UVs=AppendVector(A=Multiply(A=@n199, B=@n200), B=Constant())) {
   parameterName=LightningTimeline
   samplerType=SAMPLERTYPE_Masks
   texture=/Game/Gadgets/Assets/VinderTech_GliderChute/Textures/T_LightingTimeline_M
  }
  n202 = Add(
   A=Divide(A=ComponentMask(None=@n198.XYZ) {g=true, r=true}, B=Constant() {r=603352.375})
   B=Multiply(A=AppendVector(A=@n201.G, B=@n201.G)) {constB=5.0061}
  )
  n203 = Multiply(
   A=Clamp(None=Multiply(A=@n196, B=@n197))
   B=Power(
    Base=TextureSampleParameter2D(UVs=@n202) {parameterName=LightningTexture, samplerType=SAMPLERTYPE_LinearColor, texture=/Game/Athena/SafeZone/Textures/T_Clouds_Lightning}.G
    Exp=Constant() {r=4}
   ) {constExponent=4}
  )
  n204 = Add(A=Time(), B=Constant() {r=1.0611})
  n205 = Constant() {r=0.2238}
  n206 = CurveAtlasRowParameter(CurveTime=Frac(None=Multiply(A=@n204, B=@n205))) {
   atlas=&AthenaCurveAtlas01
   curve=/Game/Athena/SafeZone/Textures/StormLighting
   defaultValue=18
   parameterName=StormLightningCurve
  }
  n207 = Multiply(
   A=Multiply(A=Multiply(A=@n203, B=@n206.R), B=VectorParameter() {parameterName=LightningColor, defaultValue=<rgba 111.0288 100 500 1>}.RGB) {constB=10}
   B=@n165
  )
  n208 = Add(
   A=Add(A=QualitySwitch(Default=@n126, High=@n168, Epic=@n168), B=Add(A=@n192, B=@n172) {constB=0}) {constB=0}
   B=QualitySwitch(Default=@n193, Low=@n193, High=@n207, Medium=@n207, Epic=@n207)
  ) {constB=0}
  n209 = Power(Base=@n208, Exp=Constant() {r=4})
  n210 = Constant() {r=0.2}
  n211 = StaticSwitchParameter(
   True=StaticSwitchParameter(True=StaticSwitchParameter(True=@n108, False=@n106) {parameterName=DebugSHDiffuseWrapOrBackScattering}, False=Add(A=@n120, B=@n125)) {parameterName=DebugSHTerms}
   False=Add(A=@n208, B=Multiply(A=@n209, B=@n210))
  ) {parameterName=DebugLighting}
  MaterialFunctionCall!(Default=@n211, Low/Mobile=Multiply(A=@n211, B=ScalarParameter() {defaultValue=1.8, parameterName=MobileEmissiveScale})) {materialFunction=/Game/Effects/Fort_Effects/Materials/_Functions/MF_LowMobile_PathingSwitch}
  n212 = MaterialFunctionCall(Vector or UVs=@n19.XY) {materialFunction=&VectorToRadialValue}
  n213 = ComponentMask(None=@n212.Radial Coordinates) {r=true}
  n214 = ScalarParameter() {defaultValue=74, parameterName=RadialFrequency_2}
  n215 = Add(A=Sine(None=Multiply(A=@n213, B=@n214)), B=ScalarParameter() {defaultValue=0.08, parameterName=RadialBias_2})
  n216 = Multiply(A=Saturate(None=Divide(A=@n215)), B=ScalarParameter() {defaultValue=-0.152, parameterName=RadialStrength_2})
  n217 = ScalarParameter() {defaultValue=22, parameterName=RadialFrequency_1}
  n218 = Add(A=Sine(None=Multiply(A=@n213, B=@n217)), B=ScalarParameter() {defaultValue=0.5004, parameterName=RadialBias_1})
  n219 = Multiply(A=Saturate(None=Divide(A=@n218)), B=ScalarParameter() {defaultValue=-0.6599, parameterName=RadialStrength_1})
  Multiply(A=@n216, B=@n219)
  n220 = Add(A=@n216, B=@n219)
  n221 = ScalarParameter() {defaultValue=80, parameterName=RadialFrequency_3}
  n222 = Add(A=Sine(None=Multiply(A=@n213, B=@n221)), B=ScalarParameter() {defaultValue=-0.256, parameterName=RadialBias_3})
  n223 = Multiply(A=Saturate(None=Divide(A=@n222)), B=ScalarParameter() {defaultValue=-0.4285, parameterName=RadialStrength_3})
  n224 = ComponentMask(None=@n212.Radial Coordinates) {g=true}
  n225 = ScalarParameter() {defaultValue=59951.5625, parameterName=StepSize}
  n226 = Frac(None=Add(A=Add(A=@n220, B=@n223), B=Divide(A=@n224, B=@n225)))
  Multiply(A=@n226, B=ScalarParameter() {defaultValue=1423.5464, parameterName=PDO})
  n227 = LocalPosition()
  n228 = Constant()
  Multiply(
   A=Normalize(VectorInput=AppendVector(A=@n227.XY, B=@n228))
   B=LinearInterpolate(
    A=ScalarParameter() {defaultValue=31847.3477, parameterName=TempRadius1}
    B=ScalarParameter() {defaultValue=821201.0625, parameterName=TempRadius2}
    Alpha=VertexColor().R
   )
  )
  n229 = Subtract(A=CameraPositionWS())
  n230 = MaterialFunctionCall() {materialFunction=&TangentBasis}
  Multiply(A=Normalize(VectorInput=CrossProduct(A=@n229, B=@n230.Y V)))
  Multiply(A=@n2, B=@n11)
  n231 = Subtract(A=@n226, B=ScalarParameter() {defaultValue=0.0293, parameterName=LineThickness})
  n232 = VectorParameter() {parameterName=Color_WE, defaultValue=<rgba 0.3438 0.3438 0.3438 1>}
  Multiply(A=Saturate(None=Ceil(None=@n231)), B=@n232.RGB)
  NamedRerouteDeclaration(None=TextureObjectParameter() {parameterName=CloudTexture, samplerType=SAMPLERTYPE_LinearColor, texture=/Game/Athena/SafeZone/S11/T_TilingCloudsN_01}) {name=CloudTexture, nodeColor=<rgba 1 0 0.6562 1>}
  n233 = Abs(None=ComponentMask(None=@n51) {g=true})
  n234 = Abs(None=ComponentMask(None=@n51) {r=true})
  n235 = Saturate(None=Multiply(A=Add(A=@n233, B=@n234), B=ScalarParameter() {defaultValue=0.02}))
  NamedRerouteDeclaration(None=@n235) {name=Background Falloff Mask, nodeColor=<rgba 0.4766 0 1 1>}
  n236 = ScalarParameter() {parameterName=CylindricalVBias}
  NamedRerouteDeclaration(None=Add(
    A=MaterialFunctionCall!(BoundsScale=@n28) {materialFunction=/Game/Athena/SafeZone/S11/MF_CylindricalUVs}.VertexUV_SeamFix
    B=MaterialFunctionCall(G=@n236) {materialFunction=&AppendMany}.RG
   )) {name=CylindricalUVs, nodeColor=<rgba 1 0.8203 0 1>}
  n237 = SphereMask(A=ComponentMask(None=NamedRerouteUsage()) {g=true}, B=Constant() {r=0.25}, Radius=Constant() {r=0.2}, Hardness=Constant() {r=0.4})
  NamedRerouteDeclaration(None=Saturate(None=Add(A=@n95, B=@n237))) {name=StormAlphaWithGradient, nodeColor=<rgba 0.5469 0 1 1>}
  n238 = WorldPosition()
  n239 = Normalize(VectorInput=@n238.XYZ)
  n240 = ScalarParameter() {defaultValue=150000, parameterName=CameraPush}
  n241 = CameraPositionWS()
  n242 = ScalarParameter() {defaultValue=100000, parameterName=CameraPushFalloff}
  n243 = Constant()
  n244 = Multiply(
   A=Multiply(A=Multiply(A=@n239, B=@n240), B=SphereMask(A=@n238.XYZ, B=@n241, Radius=@n242, Hardness=@n243))
   B=MaterialFunctionCall(
    Input=ComponentMask(None=@n241) {b=true}
    Input Low=Constant() {r=-2400}
    Input High=Constant() {r=30000}
    Target Low=Constant()
    Target High=Constant() {r=1}
   ) {materialFunction=&RemapValueRange}.Result
  )
  NamedRerouteDeclaration(None=Multiply(A=Add(A=@n157, B=@n244), B=ScalarParameter() {group=WPO, parameterName=EnableWPO})) {name=WPO, nodeColor=<rgba 0.1016 0 1 1>}
  n245 = MaterialFunctionCall(Vector or UVs=@n20.XY) {materialFunction=&VectorToRadialValue}
  n246 = ComponentMask(None=@n245.Radial Coordinates) {r=true}
  n247 = ScalarParameter() {defaultValue=17, parameterName=Tiling_Radial}
  n248 = Divide(A=ComponentMask(None=@n245.Radial Coordinates) {g=true}, B=ScalarParameter() {defaultValue=39402.5781, parameterName=Tiling_Distance})
  n249 = TextureSample(UVs=AppendVector(A=Multiply(A=@n246, B=@n247), B=Frac(None=@n248))) {texture=&/Storm_Stylized_Cloud_M}
  NamedRerouteDeclaration(None=@n249.R) {name=Mask, nodeColor=<rgba 0 0.0391 1 1>}
  NamedRerouteDeclaration(None=@n249.G) {name=Outline, nodeColor=<rgba 0 0.0391 1 1>}
  NamedRerouteDeclaration(None=@n249.B) {name=ShadingGradient, nodeColor=<rgba 0 0.0391 1 1>}
  NamedRerouteUsage()
  Normalize(VectorInput=@n229)
  n250 = Multiply(A=ScalarParameter() {defaultValue=16.7742, parameterName=Cloud UV_Offset}, B=VertexColor().R)
  n251 = Constant()
  n252 = ScalarParameter() {defaultValue=7, parameterName=Cloud Tiling}
  n253 = ScalarParameter() {defaultValue=12, parameterName=Cloud Tiling_B}
  n254 = VertexColor()
  n255 = AppendVector(A=Round(None=LinearInterpolate(A=@n252, B=@n253, Alpha=@n254.R)), B=Constant() {r=1})
  n256 = TextureCoordinate()
  n257 = ScalarParameter() {defaultValue=0.02, parameterName=Cloud Pan Speed A}
  n258 = ScalarParameter() {defaultValue=0.01, parameterName=Cloud Pan Speed B}
  n259 = VertexColor()
  Panner(
   Coordinate=Add(A=AppendVector(A=@n250, B=@n251), B=Multiply(A=@n255, B=@n256))
   Speed=AppendVector(A=LinearInterpolate(A=@n257, B=@n258, Alpha=@n259.R), B=Constant())
  )
  Power(Base=@n235)
  n260 = ComponentMask(None=NamedRerouteUsage()) {g=true}
  n261 = Power(Base=Multiply(A=Add(A=@n105, B=@n260)) {constB=0.5}, Exp=Constant() {r=0.5})
  n262 = MaterialFunctionCall(Alpha=@n154, Min=Add(A=@n261) {constB=-0.15}, Max=@n261) {materialFunction=&SmoothStep}
  n263 = Constant() {r=-0.15}
  n264 = NamedRerouteUsage()
  n265 = Saturate(None=Multiply(A=Add(A=@n264, B=@n263), B=Constant() {r=4}))
  n266 = ComponentMask(None=NamedRerouteUsage()) {g=true}
  n267 = Add(A=OneMinus(None=@n266), B=Constant() {r=0.11})
  n268 = Constant() {r=0.92}
  n269 = Constant() {desc=-0.1, r=-0.1}
  n270 = Add(A=Max(A=@n266, B=Multiply(A=@n267, B=@n268) {constB=1.2}), B=Add(A=Constant() {desc=-0.75, r=-0.75}, B=Multiply(A=@n264, B=@n269)))
  n271 = Constant() {desc=25, r=25}
  n272 = OneMinus(None=Saturate(None=Multiply(A=@n270, B=@n271)))
  n273 = MaterialFunctionCall() {materialFunction=/Engine/Functions/Engine_MaterialFunctions02/Texturing/3DSandMayaUVCoordinates}
  n274 = Add(A=Multiply(A=@n185, B=Constant() {r=2}), B=OneMinus(None=ComponentMask(None=@n273.UVChannel0) {r=true}))
  n275 = Min(A=@n262.Result, B=Multiply(A=Max(A=@n265, B=@n272), B=Saturate(None=@n274)))
  Reroute(None=@n275) {desc=Opacity}
  n276 = NamedRerouteUsage()
  n277 = Multiply(A=Add(A=ComponentMask(None=@n276) {g=true}, B=Constant() {r=-0.735}), B=Constant() {r=20})
  n278 = StaticSwitch(
   True=OneMinus(None=Saturate(None=@n277))
   False=Constant() {r=1}
   Value=StaticBoolParameter() {defaultValue=true, parameterName=EnableTopMask}
  )
  Reroute(None=Multiply(A=@n275, B=Min(A=@n278, B=@n262.Result))) {desc=Opacity Mask}
  Reroute(None=NamedRerouteUsage()) {desc=WPO}
  n279 = LinearInterpolate(A=VectorParameter() {parameterName=Color_Shading_WE, defaultValue=<rgba 0.9219 0.9219 0.9219 1>}.RGB, Alpha=NamedRerouteUsage())
  n280 = LinearInterpolate(A=@n232.RGB, B=VectorParameter() {parameterName=Color_2_WE, defaultValue=<rgba 0.1667 0.1667 0.1667 1>}.RGB, Alpha=VertexColor().R)
  Reroute(None=Multiply(A=NamedRerouteUsage(), B=Multiply(A=@n279, B=@n280))) {desc=Emissive}
  Saturate()
  Subtract(A=@n14)
  TextureObjectParameter() {
   parameterName=CloudTexturessss
   samplerType=SAMPLERTYPE_Normal
   texture=/Game/Global/Textures/Default/Tileable_Patterns/T_Tilling_Distortion_Map_01_n
  }
  VertexColor()
  VertexInterpolator(VS=@n160) {desc=Disconnected FORT-514115}
  VertexNormalWS()
 ]
}
