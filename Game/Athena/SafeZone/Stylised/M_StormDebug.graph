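# M_StormDebug is an unlit debug material for displaying storm safe zone parameters as on-surface numbers.
# It packs SafeZoneScale, CloudStartingBias and PreventStormShrinking from the MPC_SafeZoneClouds parameter collection into a Vector3 and feeds it to the engine's DebugFloat3Values function.
# The text UVs are mirrored horizontally by negating U and wrapping with Frac.
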Material /Game/Athena/SafeZone/Stylised/M_StormDebug
graph {
 expressions {
  MaterialExpressionCollectionParameter_0 {
   class=MaterialExpressionCollectionParameter
   inputs []
   properties {collection=/Game/Athena/SafeZone/MPC_SafeZoneClouds, parameterName=SafeZoneScale}
  }
  MaterialExpressionCollectionParameter_2 {
   class=MaterialExpressionCollectionParameter
   inputs []
   properties {collection=&MPC_SafeZoneClouds, parameterName=CloudStartingBias}
  }
  MaterialExpressionCollectionParameter_3 {
   class=MaterialExpressionCollectionParameter
   inputs []
   properties {collection=&MPC_SafeZoneClouds, parameterName=PreventStormShrinking}
  }
  MaterialExpressionConstant2Vector_0 {class=MaterialExpressionConstant2Vector, properties {g=1, r=-1}}
  MaterialExpressionConvert_0 {
   class=MaterialExpressionConvert
   inputs [
    None {from=MaterialExpressionCollectionParameter_0}
    None {from=MaterialExpressionCollectionParameter_2}
    None {from=MaterialExpressionCollectionParameter_3}
   ]
   properties {
    nodeName=Make Vector3
    convertInputs [
     {
      type=Scalar
      expressionInput {
       expression=&self:MaterialExpressionCollectionParameter_0
       inputName=None
       mask=0
       maskA=0
       maskB=0
       maskG=0
       maskR=0
       outputIndex=0
      }
     }
     {
      type=Scalar
      expressionInput {
       expression=&self:MaterialExpressionCollectionParameter_2
       inputName=None
       mask=0
       maskA=0
       maskB=0
       maskG=0
       maskR=0
       outputIndex=0
      }
     }
     {
      type=Scalar
      expressionInput {
       expression=&self:MaterialExpressionCollectionParameter_3
       inputName=None
       mask=0
       maskA=0
       maskB=0
       maskG=0
       maskR=0
       outputIndex=0
      }
     }
    ]
    convertOutputs [{type=Vector3}]
   }
  }
  MaterialExpressionFrac_0 {class=MaterialExpressionFrac, inputs [None {from=MaterialExpressionMultiply_0, output=""}], properties {}}
  MaterialExpressionMaterialFunctionCall_0 {
   class=MaterialExpressionMaterialFunctionCall
   inputs [
    Vector3 {from=MaterialExpressionConvert_0}
    MaximumNumberOfDigits {from=""}
    UVs {from=MaterialExpressionFrac_0}
    DebugTextLocation RG_UpperRight BA_LowerLeft {from=""}
    Component Spacing {from=""}
   ]
   properties {materialFunction=/Engine/Functions/Engine_MaterialFunctions02/Utility/DebugFloat3Values}
  }
  MaterialExpressionMultiply_0 {
   class=MaterialExpressionMultiply
   inputs [
    A {from=MaterialExpressionTextureCoordinate_0}
    B {from=MaterialExpressionConstant2Vector_0}
   ]
  }
  MaterialExpressionTextureCoordinate_0 {class=MaterialExpressionTextureCoordinate}
 }
 properties {bAutomaticallySetUsageInEditor=false, shadingModel=MSM_Unlit, usedShadingModels=Unlit}
}
